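# Translucent, unlit, two-sided master material for a glowing light cone with a symbol, used in the ZombieElastic emote FX.
# Color blends a time-of-day tinted glow with a symbol texture, warped by broken-glass UVs and sampled per channel at offsets for chromatic separation.
# Named reroutes share Phase, Cracks and SymbolAlpha values.
# Opacity combines depth, distance and view-angle fades, a sphere mask, panning triplanar or UV fog, and particle or parameter alpha.
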Material /FNGameplayCues/Animation/Game/MainPlayer/Emotes/ZombieElastic/FX/M_ZombieElastic_LightCone_Master
graph {
 properties {
  bUsedWithInstancedStaticMeshes=true
  blendMode=BLEND_Translucent
  pixelDepthOffsetMode=PDOM_Legacy
  shadingModel=MSM_Unlit
  twoSided=true
  usedShadingModels=Unlit
 }
 nodes [
  n1 = ScalarParameter() {defaultValue=1, group=Color Adjustments, parameterName=Morning Multiplier}
  n2 = VectorParameter() {parameterName=GlowColor, defaultValue=<rgba 1 0.5328 0.124 1>}
  n3 = ScalarParameter() {defaultValue=1, group=Color Adjustments, parameterName=Day Multiplier}
  n4 = ScalarParameter() {defaultValue=1, group=Color Adjustments, parameterName=Evening Multiplier}
  n5 = ScalarParameter() {defaultValue=1, group=Color Adjustments, parameterName=Night Multiplier}
  n6 = Constant2Vector() {r=0.05}
  n7 = Constant2Vector()
  n8 = NamedRerouteUsage()
  n9 = TextureCoordinate()
  n10 = TextureSample() {samplerType=SAMPLERTYPE_LinearColor, texture=&/T_ZombieElasticBrokenGlass_N}
  n11 = AppendVector(A=@n10.R, B=@n10.R)
  n12 = NamedRerouteUsage()
  n13 = ScalarParameter() {defaultValue=0.85, parameterName=CutoutTextureScale}
  n14 = MaterialFunctionCall(
   UVs=LinearInterpolate(A=LinearInterpolate(A=@n9, B=@n11) {constAlpha=0.225}, B=LinearInterpolate(A=@n9, B=@n11) {constAlpha=0.125}, Alpha=@n12) {constAlpha=0.25}
   Texture Scale=LinearInterpolate(A=@n13, B=Add(A=@n13) {constB=0.05}, Alpha=@n12)
  ) {materialFunction=&ScaleUVsByCenter}
  n15 = TextureSampleParameter2D(UVs=Add(A=LinearInterpolate(A=@n6, B=@n7, Alpha=@n8) {constAlpha=0.2}, B=@n14.UVs)) {group=End Symbol, parameterName=Symbol Texture, samplerSource=SSM_Clamp_WorldGroupSettings, texture=/Game/2dAssets/Emoji/Season4/Crab}
  n16 = TextureSampleParameter2D(UVs=@n14.UVs, Apply View MipBias=ScalarParameter() {parameterName=Symbol LOD Bias}) {group=End Symbol, parameterName=Symbol Texture, samplerSource=SSM_Clamp_WorldGroupSettings, texture=&Crab}
  n17 = Constant2Vector() {r=-0.05}
  n18 = Constant2Vector()
  n19 = NamedRerouteUsage()
  n20 = TextureSampleParameter2D(UVs=Add(A=LinearInterpolate(A=@n17, B=@n18, Alpha=@n19) {constAlpha=0.2}, B=@n14.UVs)) {group=End Symbol, parameterName=Symbol Texture, samplerSource=SSM_Clamp_WorldGroupSettings, texture=&Crab}
  n21 = MaterialFunctionCall(X=@n15.R, Y=@n16.G, Z=@n20.B) {materialFunction=&MakeFloat3}
  n22 = Multiply(A=NamedRerouteUsage(), B=Multiply(A=Multiply(A=@n21.Result, B=@n16.A), B=@n14.0-1 mask))
  n23 = Constant() {r=0.2}
  LinearInterpolate(
   A=MaterialFunctionCall(
    Color_Morning=Multiply(A=@n1, B=@n2.RGB) {constB=250}
    Color_Day=Multiply(A=@n2.RGB, B=@n3) {constB=250}
    Color_Evening=Multiply(A=@n4, B=@n2.RGB) {constB=250}
    Color_Night=Multiply(A=@n5, B=@n2.RGB) {constB=250}
   ) {materialFunction=&Color_Time_of_Day}.Result
   B=LinearInterpolate(A=@n22, B=Desaturation(None=@n22, Fraction=@n23), Alpha=NamedRerouteUsage())
   Alpha=NamedRerouteUsage()
  )
  n24 = Multiply(A=NamedRerouteUsage(), B=NamedRerouteUsage())
  n25 = ScalarParameter() {parameterName=DepthFade}
  n26 = MaterialFunctionCall(
   Position=WorldPosition().XYZ
   Fade_Start_Distance=ScalarParameter() {defaultValue=250, group=Fade, parameterName="Fade_Start_Distance (S)"}
   Fade_End_Distance=ScalarParameter() {defaultValue=100, group=Fade, parameterName="Fade_End_Distance (S)"}
  ) {materialFunction=&MF_Distance_SyncedFade_FX}
  n27 = TextureCoordinate()
  n28 = Constant() {r=0.5}
  n29 = ScalarParameter() {defaultValue=0.5, group=End Symbol, parameterName=Radius}
  n30 = ScalarParameter() {defaultValue=0.5, group=End Symbol, parameterName=Hardness}
  n31 = Multiply(
   A=MaterialFunctionCall("FadeSensitivity (optional)"=ScalarParameter() {group=Fade, parameterName=Angle Fade Sensitivity}) {materialFunction=&FadeBasedOnViewAngle}.Result
   B=Power(
    Base=SphereMask(A=@n27, B=@n28, Radius=@n29, Hardness=@n30)
    Exp=ScalarParameter() {defaultValue=2, group=End Symbol, parameterName=SphereMask Power}
   )
  )
  n32 = Power(
   Base=Multiply(A=DepthFade(FadeDistance=@n25) {fadeDistanceDefault=25}, B=Multiply(A=@n26.Result, B=@n31))
   Exp=ScalarParameter() {defaultValue=1, parameterName=Final Power}
  )
  n33 = Multiply(A=Multiply(A=Saturate(None=@n24), B=Clamp(None=@n32)), B=ScalarParameter() {defaultValue=0.1, parameterName=Final Multiplier})
  n34 = AppendVector(
   A=ScalarParameter() {defaultValue=0.005, group=Fog, parameterName=Fog Pan Speed X}
   B=ScalarParameter() {group=Fog, parameterName=Fog Pan Speed Y}
  )
  n35 = VectorParameter() {group=Fog, parameterName=FogTextureTiling, defaultValue=<rgba 2 2 2 1>}
  n36 = MaterialFunctionCall!(
   CameraVector
   Time Scale
   Panner Speed=@n34
   Texture=TextureObjectParameter() {group=Fog, parameterName=Fog Texture, texture=&T_Smooth_tile}
   Tiling=@n35.RGB
   Offset
   AxisFadeContrast
  ) {materialFunction=/Game/Weapons/WeaponSkins/MaterialFunction/TriplanarCamerVector_Panning}
  n37 = TextureCoordinate()
  n38 = ComponentMask(None=@n35.RGB) {g=true, r=true}
  n39 = Multiply(A=Time(), B=@n34)
  n40 = TextureSample(UVs=Add(A=Multiply(A=@n37, B=@n38), B=Frac(None=@n39))) {texture=&T_Smooth_tile}
  n41 = LinearInterpolate(
   A=MaterialFunctionCall(
    In=StaticSwitchParameter(True=@n36.Result, False=@n40.G) {defaultValue=true, group=Fog, parameterName=UseTriPlanarFog?}
    Contrast=ScalarParameter() {parameterName=Cheap Contrast}
   ) {materialFunction=&CheapContrast}.Result
   Alpha=ScalarParameter() {defaultValue=0.5, group=Fog, parameterName=Fade Fog}
  )
  Multiply(
   A=LinearInterpolate(A=ParticleColor().A, B=ScalarParameter() {defaultValue=1, parameterName=Alpha}, Alpha=ScalarParameter() {parameterName=Use_Mesh_Alpha})
   B=Clamp(None=Multiply(A=@n33, B=@n41))
  )
  NamedRerouteDeclaration(None=Saturate(None=ScalarParameter() {defaultValue=1, parameterName=Phase})) {name=Phase, nodeColor=<rgba 0.0312 1 0 1>}
  NamedRerouteDeclaration(None=Multiply(A=@n16.A, B=@n10.B)) {name=Cracks, nodeColor=<rgba 1 0.1641 0 1>}
  NamedRerouteDeclaration(None=@n16.A) {name=SymbolAlpha, nodeColor=<rgba 0 1 0.4844 1>}
 ]
}
